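# Physics asset for the PureCereal pickaxe mesh, a two-handed melee weapon cosmetic.
# It holds a single body on the Melee_TwoHanded_Handle bone, shaped as one offset, rotated capsule.
# There are no constraints, so the asset gives the pickaxe simple collision with no articulated parts.
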PhysicsAsset /BRCosmetics/Weapons/FORT_Melee/Pickaxe_PureCereal/Meshes/PureCereal_Axe_PhysicsAsset
bodies {Melee_TwoHanded_Handle [{shapeType=Capsule, length=24.2816, radius=4.243, center=<-0.538 -0.5589 12.7605>, rotation=<rot 2.5181 -90 0.0001>}]}
constraints []
